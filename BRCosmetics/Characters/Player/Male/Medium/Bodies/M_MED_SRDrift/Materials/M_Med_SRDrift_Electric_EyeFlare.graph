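# Translucent unlit material for the electric eye flare on the SRDrift character, using the Sprite function and a camera-direction offset to build a billboarded glow.
# Opacity combines a radial sphere mask, multiplied-sine flicker and two panned ramp samples with noise-randomized speeds, softened by depth fade.
# Emissive pulses a variant color over time with a linked hue shift and HDR remap, masked by the same electrical pattern.
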Material /BRCosmetics/Characters/Player/Male/Medium/Bodies/M_MED_SRDrift/Materials/M_Med_SRDrift_Electric_EyeFlare
graph {
 properties {
  bUsedWithMeshDeformer=true
  bUsedWithMorphTargets=true
  bUsedWithSkeletalMesh=true
  blendMode=BLEND_Translucent
  pixelDepthOffsetMode=PDOM_Legacy
  shadingModel=MSM_Unlit
  twoSided=true
  usedShadingModels=Unlit
 }
 nodes [
  n1 = VectorParameter() {parameterName=SpriteScale, defaultValue=<rgba 8 6 0 0>}
  Add(
   A=MaterialFunctionCall!(Center Location=WorldPosition().XYZ, 0-1 UVs, XY scale=ComponentMask(None=@n1.RGB) {g=true, r=true}, OverridePivot, "Normals (Optional)") {materialFunction=&Sprite}.World position offset
   B=Multiply(
    A=MaterialFunctionCall() {materialFunction=&CameraDirectionVector}.Result
    B=ScalarParameter() {defaultValue=-5, parameterName=CameraOffset}
   )
  )
  n2 = ScalarParameter() {defaultValue=500, parameterName=NearFade}
  Clamp(None=Power(Base=MaterialFunctionCall(Distance=@n2) {materialFunction=&Near_Camera_Fade}.Result))
  ComponentMask(None=VectorParameter() {defaultValue=<rgba 1000 0 0 1>}.RGB) {b=true, g=true, r=true}
  Constant() {r=1}
  n3 = SphereMask(
   A=TextureCoordinate()
   B=Constant2Vector() {g=0.5, r=0.5}
   Radius=ScalarParameter() {defaultValue=0.45, parameterName=SphereMask_Radius}
   Hardness=ScalarParameter() {defaultValue=-1, parameterName=SphereMask_Hardness}
  )
  n4 = ScalarParameter() {defaultValue=2.25, parameterName=HotCenter_Eye}
  n5 = Sine(None=Time()) {period=1.4}
  n6 = Sine(None=Time()) {period=2.1}
  n7 = Time()
  n8 = LinearInterpolate(Alpha=Multiply(A=Multiply(A=@n5, B=@n6), B=Sine(None=@n7) {period=2.7}))
  n9 = ScalarParameter() {defaultValue=3, parameterName=SinePower}
  n10 = Saturate(None=Multiply(A=Power(Base=@n8, Exp=@n9) {constExponent=6}) {constB=25})
  n11 = TextureCoordinate() {coordinateIndex=1}
  n12 = Multiply(A=TextureCoordinate() {coordinateIndex=1, uTiling=0}, B=ScalarParameter() {defaultValue=16, parameterName=NumberOfBars})
  n13 = MaterialFunctionCall!(
   PrimaryTextureCoordinates=@n12
   2ndOrderTextureCoordinates=@n12
   RandomSeed=ScalarParameter() {defaultValue=1, parameterName=Speed}
   NoiseDensityBias
   Bypass Noise Influence Crosstalk
  ) {materialFunction=/Game/Effects/Fort_Effects/Materials/_Functions/MF_RandomUVSample}
  n14 = ComponentMask(None=Multiply(A=VectorParameter() {parameterName=ElectricalMask_Speed, defaultValue=<rgba 0 5 0 0>}.RGB, B=@n13.2ndOrderNoiseWithRange)) {g=true, r=true}
  n15 = TextureCoordinate() {coordinateIndex=1}
  n16 = ComponentMask(None=Multiply(A=@n13.2ndOrderNoiseWithRange, B=VectorParameter() {parameterName=ElectricalMask_Speed_2, defaultValue=<rgba 0 4.1 0 0>}.RGB)) {g=true, r=true}
  n17 = Multiply(
   A=TextureSample(UVs=Panner(Coordinate=@n11, Speed=@n14) {speedY=0.7}) {texture=/Game/Global/Textures/Default/Masks_and_Gradients/T_Water_Top_Ramp}.G
   B=TextureSample(UVs=Panner(Coordinate=@n15, Speed=@n16) {speedY=1}) {texture=&T_Water_Top_Ramp}.G
  )
  DepthFade(
   Opacity=Multiply(A=Power(Base=@n3, Exp=@n4), B=Multiply(A=@n10, B=@n17))
   FadeDistance=ScalarParameter() {defaultValue=5, parameterName=FadeDistance}
  )
  n18 = Time()
  n19 = LinearInterpolate(
   A=ScalarParameter() {defaultValue=0.3, parameterName=MinColor}
   B=ScalarParameter() {defaultValue=0.8, parameterName=MaxColor}
   Alpha=Clamp(None=Sine(None=@n18) {period=0.5})
  )
  n20 = Clamp(None=@n19)
  n21 = VectorParameter() {parameterName=VariantColor, defaultValue=<rgba 1 1 1 1>}
  n22 = MaterialFunctionCall(
   Hue Shift Percentage=Multiply(A=@n19, B=ScalarParameter() {defaultValue=0.175, parameterName=HueShift_Mult})
   Texture=Multiply(A=@n19, B=Multiply(A=@n21.RGB) {constB=150})
  ) {materialFunction=&HueShift}
  MaterialFunctionCall(Emissive=Multiply(
    A=MaterialFunctionCall(LDR Input=@n20, HDR Tint=@n22.Result) {materialFunction=&DeriveHDRfromLDR}.RemappedRange
    B=@n17
   )) {materialFunction=&MF_FX_EmissiveAdjustments}
  ObjectPositionWS()
 ]
}
