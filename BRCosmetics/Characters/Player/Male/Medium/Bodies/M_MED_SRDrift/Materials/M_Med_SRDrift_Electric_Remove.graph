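# Translucent unlit skeletal mesh material for the SRDrift player body's electric variant, apparently meant to hide that part of the mesh.
# It multiplies each point's offset from the object position by a Scaling parameter, whose negative default would collapse the mesh to that position if used as world position offset.
# A zero constant feeds MF_FX_EmissiveAdjustments, and another zero constant stands alone, with no output connections shown.
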Material /BRCosmetics/Characters/Player/Male/Medium/Bodies/M_MED_SRDrift/Materials/M_Med_SRDrift_Electric_Remove
graph {
 properties {
  bUsedWithMeshDeformer=true
  bUsedWithMorphTargets=true
  bUsedWithSkeletalMesh=true
  blendMode=BLEND_Translucent
  pixelDepthOffsetMode=PDOM_Legacy
  shadingModel=MSM_Unlit
  usedShadingModels=Unlit
 }
 nodes [
  Constant()
  MaterialFunctionCall(Emissive=Constant()) {materialFunction=&MF_FX_EmissiveAdjustments}
  Multiply(A=Subtract(A=WorldPosition().XYZ, B=ObjectPositionWS()), B=ScalarParameter() {defaultValue=-1, parameterName=Scaling})
 ]
}
