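# Physics asset of simplified body collision shapes for the M_MED_SRDrift male player body, named as colliders for its coat.
# Capsules cover the head, neck, upper spine, clavicles, upper arms, thighs and calves, with a sphere at the pelvis; forearms, hands and feet have no bodies.
# Constraints chain these bodies out from the pelvis with free swing and twist, so it appears intended as cloth collision rather than a ragdoll.
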PhysicsAsset /BRCosmetics/Characters/Player/Male/Medium/Bodies/M_MED_SRDrift/Meshes/M_MED_SRDrift_CoatColliders
bodies {
 calf_l [{shapeType=Capsule, length=43.742, radius=6.9116, center=<-25.4904 -2.5817 0.2436>, rotation=<rot -0.1958 91.6743 -89.4989>}]
 calf_r [{shapeType=Capsule, length=44.7862, radius=8.0971, center=<24.7067 1.4625 -0.1801>, rotation=<rot -0.253 90.5446 -89.4123>}]
 clavicle_l [{shapeType=Capsule, length=10.3261, radius=8.5802, center=<8.2822 -0.2527 -2.0684>, rotation=<rot 90 -0.0049 179.9951>}]
 clavicle_r [{shapeType=Capsule, length=10.3261, radius=8.5802, center=<-7.9705 0.6201 2.6048>, rotation=<rot 90 0.0198 0.0198>}]
 head [{shapeType=Capsule, length=0, radius=18, center=<11.6361 0.0962 0>, rotation=<rot 0 -90 0>}]
 neck_02 [{shapeType=Capsule, length=0.505, radius=10.755, rotation=<rot 0 -90 0>}]
 pelvis [{shapeType=Sphere, radius=13.8856, center=<6.8561 0.1415 0.0004>}]
 spine_03 [{shapeType=Capsule, length=16.0169, radius=12.5072, center=<-0.4494 -1.3822 0.0003>, rotation=<rot -83.7194 -89.0844 89.9913>}]
 spine_05 [{shapeType=Capsule, length=16.7232, radius=12.517, center=<-2.2432 -2.2678 0>}]
 thigh_l [{shapeType=Capsule, length=39.6153, radius=10.9315, center=<-18.3735 -0.0001 0>, rotation=<rot 90 0 0>}]
 thigh_r [{shapeType=Capsule, length=39.6153, radius=10.9315, center=<19.0128 0 0>, rotation=<rot 90 0 0>}]
 upperarm_l [{shapeType=Capsule, length=18.8261, radius=7.5802, center=<12.1824 -0.2526 -0.6865>, rotation=<rot 90 0.005 -179.995>}]
 upperarm_r [{shapeType=Capsule, length=17.1357, radius=7.1339, center=<-12.8316 -0.0074 -0.1319>, rotation=<rot 0 -88.4734 -86.4203>}]
}
constraints [
 {
  bone1Name=spine_05
  bone2Name=spine_03
  swing1LimitDegrees=45
  swing1Motion=Free
  swing2LimitDegrees=45
  swing2Motion=Free
  twistLimitDegrees=45
  twistMotion=Free
 }
 {
  bone1Name=clavicle_l
  bone2Name=spine_05
  swing1LimitDegrees=45
  swing1Motion=Free
  swing2LimitDegrees=45
  swing2Motion=Free
  twistLimitDegrees=45
  twistMotion=Free
 }
 {
  bone1Name=upperarm_l
  bone2Name=clavicle_l
  swing1LimitDegrees=45
  swing1Motion=Free
  swing2LimitDegrees=45
  swing2Motion=Free
  twistLimitDegrees=45
  twistMotion=Free
 }
 {
  bone1Name=clavicle_r
  bone2Name=spine_05
  swing1LimitDegrees=45
  swing1Motion=Free
  swing2LimitDegrees=45
  swing2Motion=Free
  twistLimitDegrees=45
  twistMotion=Free
 }
 {
  bone1Name=upperarm_r
  bone2Name=clavicle_r
  swing1LimitDegrees=45
  swing1Motion=Free
  swing2LimitDegrees=45
  swing2Motion=Free
  twistLimitDegrees=45
  twistMotion=Free
 }
 {
  bone1Name=calf_l
  bone2Name=thigh_l
  swing1LimitDegrees=45
  swing1Motion=Free
  swing2LimitDegrees=45
  swing2Motion=Free
  twistLimitDegrees=45
  twistMotion=Free
 }
 {
  bone1Name=calf_r
  bone2Name=thigh_r
  swing1LimitDegrees=45
  swing1Motion=Free
  swing2LimitDegrees=45
  swing2Motion=Free
  twistLimitDegrees=45
  twistMotion=Free
 }
 {
  bone1Name=head
  bone2Name=neck_02
  swing1LimitDegrees=45
  swing1Motion=Free
  swing2LimitDegrees=45
  swing2Motion=Free
  twistLimitDegrees=45
  twistMotion=Free
 }
 {
  bone1Name=neck_02
  bone2Name=spine_05
  swing1LimitDegrees=45
  swing1Motion=Free
  swing2LimitDegrees=45
  swing2Motion=Free
  twistLimitDegrees=45
  twistMotion=Free
 }
 {
  bone1Name=spine_03
  bone2Name=pelvis
  swing1LimitDegrees=45
  swing1Motion=Free
  swing2LimitDegrees=45
  swing2Motion=Free
  twistLimitDegrees=45
  twistMotion=Free
 }
 {
  bone1Name=thigh_l
  bone2Name=pelvis
  swing1LimitDegrees=45
  swing1Motion=Free
  swing2LimitDegrees=45
  swing2Motion=Free
  twistLimitDegrees=45
  twistMotion=Free
 }
 {
  bone1Name=thigh_r
  bone2Name=pelvis
  swing1LimitDegrees=45
  swing1Motion=Free
  swing2LimitDegrees=45
  swing2Motion=Free
  twistLimitDegrees=45
  twistMotion=Free
 }
]
