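# Post-copy Anim Blueprint for the SRDrift tier 01-04 male body, adding procedural secondary motion to the mask, hair and hoody collar.
# It copies the owner mesh's pose, counter-rolls mask_dyn from head-to-spine rotation, then caches the result.
# Anim Dynamics then moves hair and collars, with recent firing adding hair lift and randomized hand-to-head collar force.
# Playlists using custom character parts instead output the cached pose with the head scaled near zero.
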AnimBlueprint /BRCosmetics/Characters/Player/Male/Medium/Bodies/M_MED_SRDrift/Meshes/M_MED_SRDrift_tier0104_AnimBP
target_skeleton=&/M_MED_SRDrift_Skeleton
graphs [
 AnimGraph {nodes [
   n1 = "Transform (Modify) Bone - Bone: head"!(ComponentPose=Local To Component(LocalPose=Use cached pose 'PreScale Cache'()), Alpha=1) -> (Pose) {scaleMode=BMM_Replace, alphaBoolBlend {blendOption=Linear}, boneToModify {boneName=head}, scale=<0.01 0.01 0.01>}
   n2 = Get bRecentlyFired() -> (bRecentlyFired:bool)
   n3 = Anim Dynamics - Bone: hair_dyn(
    ComponentPose=Local To Component(LocalPose=Use cached pose 'PreScale Cache'())
    Alpha=0.2
    ExternalForce=MakeVector(Y=0, Z=SelectFloat(A=10000, B=0, bPickA=@n2))
   ) {
    angularDampingOverride=0.99
    angularSpringConstant=3000
    bAngularSpring=true
    bGravityOverrideInSimSpace=true
    bOverrideAngularDamping=true
    numSolverIterationsPostUpdate=4
    numSolverIterationsPreUpdate=16
    simulationSpace=Actor
    alphaBoolBlend {blendOption=Linear}
    boundBone {boneName=hair_dyn}
    physicsBodyDefinitions [{
      sphereCollisionRadius=0
      boundBone {boneName=hair_dyn}
      boxExtents=<15 18 18>
      constraintSetup {angularLimitsMax=<5 20 20>, angularLimitsMin=<-5 -20 -20>, angularTarget=<1 0 0>}
      localJointOffset=<-3 0 0>
     }]
   }
   n4 = Get recently_fired_force_hoody() -> (recently_fired_force_hoody:Vector)
   n5 = Anim Dynamics - Bone: collar_l(ComponentPose=@n3, Alpha=0.7, ExternalForce=@n4) {
    angularDampingOverride=0.8
    angularSpringConstant=500
    bAngularSpring=true
    bGravityOverrideInSimSpace=true
    bOverrideAngularDamping=true
    numSolverIterationsPostUpdate=4
    numSolverIterationsPreUpdate=16
    alphaBoolBlend {blendOption=Linear}
    boundBone {boneName=collar_l}
    componentLinearAccScale=<1 1 1>
    componentLinearVelScale=<2 2 2>
    physicsBodyDefinitions [{
      sphereCollisionRadius=0
      boundBone {boneName=collar_l}
      boxExtents=<15 18 18>
      constraintSetup {angularLimitsMax=<5 20 20>, angularLimitsMin=<-5 -20 -20>, angularTarget=<1 0 0>, linearAxesMax=<0 0 1>, linearAxesMin=<0 0 -1>}
      localJointOffset=<-3 0 0>
     }]
   }
   n6 = Anim Dynamics - Bone: collar_r(ComponentPose=@n5, Alpha=0.7, ExternalForce=@n4) {
    angularDampingOverride=0.8
    angularSpringConstant=500
    bAngularSpring=true
    bGravityOverrideInSimSpace=true
    bOverrideAngularDamping=true
    numSolverIterationsPostUpdate=4
    numSolverIterationsPreUpdate=16
    alphaBoolBlend {blendOption=Linear}
    boundBone {boneName=collar_r}
    componentLinearAccScale=<1 1 1>
    componentLinearVelScale=<2 2 2>
    physicsBodyDefinitions [{
      sphereCollisionRadius=0
      boundBone {boneName=collar_r}
      boxExtents=<15 18 18>
      constraintSetup {angularLimitsMax=<5 20 20>, angularLimitsMin=<-5 -20 -20>, angularTarget=<1 0 0>, linearAxesMax=<0 0 1>, linearAxesMin=<0 0 -1>}
      localJointOffset=<-3 0 0>
     }]
   }
   Output Pose(Result=Blend Poses by bool(
     BlendPose_0=Component To Local(ComponentPose=@n1)
     BlendPose_1=Component To Local(ComponentPose=@n6)
     BlendTime_0=0
     BlendTime_1=0
     bActiveValue=Get GetPlayerListUsesCustomCharacterParts() -> (GetPlayerListUsesCustomCharacterParts:bool)
    ) {blendType=Linear}) {layerGroup=None}
   n7 = Get MeshToCopy() -> (MeshToCopy:Skeletal Mesh Component Object Reference)
   n8 = "Transform (Modify) Bone - Bone: mask_dyn"(
    ComponentPose=Local To Component(LocalPose=Copy Pose From Mesh(SourceMeshComponent=@n7))
    Alpha=Get headSideToSide() -> (headSideToSide:double)
    Translation="0,0,0"
    Rotation="0,0,-15"
    Scale="1,1,1"
   ) {
    rotationMode=BMM_Additive
    rotationSpace=BCS_BoneSpace
    translationMode=BMM_Additive
    translationSpace=BCS_BoneSpace
    alphaBoolBlend {blendOption=Linear}
    alphaScaleBiasClamp {bMapRange=true, inRange {max=60}}
    boneToModify {boneName=mask_dyn}
    rotation {roll=-15}
   }
   n9 = "Transform (Modify) Bone - Bone: mask_dyn"(ComponentPose=@n8, Alpha=Get headSideToSide(), Translation="0,0,0", Rotation="0,0,15", Scale="1,1,1") {
    rotationMode=BMM_Additive
    rotationSpace=BCS_BoneSpace
    translationMode=BMM_Additive
    translationSpace=BCS_BoneSpace
    alphaBoolBlend {blendOption=Linear}
    alphaScaleBiasClamp {bMapRange=true, inRange {max=0, min=-60}, outRange {max=0, min=1}}
    boneToModify {boneName=mask_dyn}
    rotation {roll=15}
   }
   Save cached pose 'PreScale Cache'(Pose=Component To Local(ComponentPose=@n9)) {cacheName=PreScale Cache}
  ]}
 EventGraph {nodes [
   Event BlueprintUpdateAnimation()
   > n1 = Sequence()
   n2 = Set MeshToCopy(execute=@n7.then, MeshToCopy:Skeletal Mesh Component Object Reference=GetOwnerSkeletalMeshComponent()) -> (Output_Get:Skeletal Mesh Component Object Reference)
   Set gun_to_head_vec3(
    execute=@n1.then_1
    gun_to_head_vec3:Vector=vector - vector!(
     A:Vector=GetSocketLocation!(self:Scene Component Object Reference=@n2, InSocketName:Name=head) -> (ReturnValue:Vector)
     B:Vector=GetSocketLocation(self=@n2, InSocketName=hand_r)
    ) -> (ReturnValue:Vector) {is=title K2Node_CallFunction}
   ) -> (Output_Get:Vector)
   n3 = "To Vector (Float)"!(InDouble:double=RandomFloatInRangeFromStream!(Stream:"Random Stream Structure (by ref)"=Get random_stream() -> (random_stream:Random Stream Structure), Min:float=1, Max:float=120) -> (ReturnValue:float)) -> (ReturnValue:Vector)
   n4 = Get gun_to_head_vec3() -> (gun_to_head_vec3:Vector)
   n5 = Get bRecentlyFired()
   > Set recently_fired_force_hoody(recently_fired_force_hoody:Vector=SelectVector(
     A=vector * vector(A:Vector=@n4, B:Vector=@n3) -> (ReturnValue:Vector) {is=title K2Node_CommutativeAssociativeBinaryOperator}
     B="0, 0, 0"
     bPickA=@n5
    )) -> (Output_Get:Vector)
   n6 = GetDeltaTransformFromRefPose(self=GetOwnerSkeletalMeshComponent(), BoneName=head, BaseName=spine_05)
   Set headSideToSide(execute=@n1.then_0, headSideToSide:double=BreakRotator(InRot=BreakTransform(InTransform=@n6).Rotation)) -> (Output_Get:double)
   n7 = Event BlueprintInitializeAnimation()
   n8 = RandomFloatInRangeFromStream(Min=50, Max=240)
   SelectVector(A=vector * vector(A=@n4, B="To Vector (Float)"(InDouble=@n8)), B="0, 0, 0", bPickA=@n5)
   Set GetPlayerListUsesCustomCharacterParts(execute=@n10.then, GetPlayerListUsesCustomCharacterParts:bool=GetPlaylistUsesCustomCharacterParts!() -> (ReturnValue:bool)) -> (Output_Get:bool)
   n9 = MakeRandomStream!(InitialSeed:int=0) -> (ReturnValue:Random Stream Structure)
   n10 = Set random_stream(execute=@n2.then, random_stream:Random Stream Structure=@n9) -> (Output_Get:Random Stream Structure)
  ]}
]
variables [
 GetPlayerListUsesCustomCharacterParts {type=boolean}
 gun_to_head_vec3 {type=Vector}
 headSideToSide {type=number}
 MeshToCopy {type=SkeletalMeshComponent}
 random_stream {type=RandomStream, default {initialSeed=0, seed=0}}
 recently_fired_force_hoody {type=Vector}
]
