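# Physics asset for the dynamic jacket on the tier 05/06 male medium SRDrift player body.
# Capsules cover the spine, pelvis and legs, with named attach shapes on the root.
# Box chains form the front, side and back panels, swing- and twist-limited, anchored to spine_03 or pelvis, mostly through tiny MSTR bodies.
# Angularly free constraints link the body bones and cross-tie both back tails and each side chain to its neighboring front chain.
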PhysicsAsset /BRCosmetics/Characters/Player/Male/Medium/Bodies/M_MED_SRDrift/Meshes/M_MED_SRDrift_tier0506_Dyn_jacket_Update
bodies {
 calf_l [{shapeType=Capsule, length=28.3158, radius=9.3501, center=<-19.4264 -2.7964 0>, rotation=<rot 90 0 0>}]
 calf_r [{shapeType=Capsule, length=27.0658, radius=10.6001, center=<19.4264 2.7964 0>, rotation=<rot 90 0 0>}]
 foot_l [{
   shapeType=Capsule
   length=17.5237
   radius=11.7507
   shapeName=foot_l_capsule
   center=<-3.4271 8.491 0.1091>
   rotation=<rot 0.7349 0.9586 -85.8278>
  }]
 foot_r [{
   shapeType=Capsule
   length=17.7876
   radius=15.5175
   shapeName=foot_r_capsule
   center=<3.4917 -9.9816 -0.1619>
   rotation=<rot 0.7221 1.1909 -85.9009>
  }]
 jacket_back_left01 [{shapeType=Box, extentX=6.132, extentY=25.8806, extentZ=2.5834, center=<3 0 0>}]
 jacket_back_left02 [{shapeType=Box, extentX=14.2044, extentY=25.8806, extentZ=2.5834, center=<6.5 0 0>}]
 jacket_back_left03 [{shapeType=Box, extentX=11.3423, extentY=25.8806, extentZ=2.5834, center=<6 0 0>}]
 jacket_back_left04 [{shapeType=Box, extentX=22.1375, extentY=25.8806, extentZ=2.5834, center=<12 0 0>}]
 jacket_back_right01 [{shapeType=Box, extentX=6.132, extentY=25.8806, extentZ=2.5834, center=<-3 0 0>}]
 jacket_back_right02 [{shapeType=Box, extentX=14.2044, extentY=25.8806, extentZ=2.5834, center=<-6.5 0 0>}]
 jacket_back_right03 [{shapeType=Box, extentX=11.3423, extentY=29.8806, extentZ=4.3334, center=<-6 0 0>}]
 jacket_back_right04 [{shapeType=Box, extentX=22.1375, extentY=29.8806, extentZ=4.5834, center=<-12 0 0>}]
 jacket_front_left01 [{shapeType=Box, extentX=10.25, extentY=16.6751, extentZ=2.5834, center=<-5.6132 -3.1 0.0004>}]
 jacket_front_left0103 [{shapeType=Box, extentX=12.625, extentY=24.5501, extentZ=2.5834, center=<-6.4506 -8.7 0.0001>}]
 jacket_front_left02 [{shapeType=Box, extentX=11.5625, extentY=20.0501, extentZ=2.5834, center=<-5.47 -5.5 -0.0001>}]
 jacket_front_left_dist [
  {shapeType=Capsule, length=15, radius=15, center=<0 -21.9707 -6.7529>, rotation=<rot 90 0 0>}
  {shapeType=Box, extentX=10.5, extentY=24.5501, extentZ=2.5834, center=<-5.4534 -9 0.0004>}
 ]
 jacket_front_right01 [{shapeType=Box, extentX=10.25, extentY=16.6751, extentZ=2.5834, center=<-5.6132 3.1 0.0002>}]
 jacket_front_right02 [{shapeType=Box, extentX=11.5625, extentY=20.0501, extentZ=2.5834, center=<-5.47 5.5 -0.0002>}]
 jacket_front_right03 [{shapeType=Box, extentX=12.625, extentY=24.5501, extentZ=2.5834, center=<-6.4506 8.7 0.0003>}]
 jacket_front_right_dist [
  {shapeType=Capsule, length=15, radius=15, center=<0 21.9707 -6.7529>, rotation=<rot 90 0 0>}
  {shapeType=Box, extentX=10.5, extentY=24.5501, extentZ=2.5834, center=<-5.4534 9 0.0005>}
 ]
 jacket_front_top_driver_l [{shapeType=Box, extentX=11.1875, extentY=15.4876, extentZ=2.5834, center=<-5.5287 -2.8 0.0005>}]
 jacket_front_top_driver_l_MSTR [{shapeType=Box, extentX=0.5, extentY=0.5, extentZ=2.1018}]
 jacket_front_top_driver_r [{shapeType=Box, extentX=11.1875, extentY=15.4876, extentZ=2.5834, center=<-5.5287 2.8 0.0006>}]
 jacket_front_top_driver_r_MSTR [{shapeType=Box, extentX=0.5, extentY=0.5, extentZ=2.1018}]
 jacket_left01 [{shapeType=Box, extentX=16.4375, extentY=14.3806, extentZ=3.4584, center=<-7.7321 0 0>}]
 jacket_left01_MSTR [{shapeType=Box, extentX=0.5, extentY=0.5, extentZ=2.1018}]
 jacket_left02 [{shapeType=Box, extentX=16.5625, extentY=14.3806, extentZ=2.5834, center=<-7.7947 -0.0002 0>}]
 jacket_left03 [{shapeType=Box, extentX=13.8125, extentY=14.3806, extentZ=2.5834, center=<-8.4792 0 0>}]
 jacket_left04 [{shapeType=Box, extentX=16.4375, extentY=14.3806, extentZ=3.4584, center=<-7.7321 0 0>}]
 jacket_left04_MSTR [{shapeType=Box, extentX=0.5, extentY=0.5, extentZ=2.1018}]
 jacket_left05 [{shapeType=Box, extentX=16.5625, extentY=14.3806, extentZ=2.5834, center=<-7.7947 -0.0002 0>}]
 jacket_left06 [{shapeType=Box, extentX=13.8125, extentY=19.8806, extentZ=3.5834, center=<-8.4792 0 0>}]
 pelvis [{shapeType=Capsule, length=10.2645, radius=13.8284, center=<-4.4394 -2.7986 -0.0001>}]
 root [
  {shapeType=Sphere, radius=187.1686, shapeName=attach_sphere, center=<-3.4923 -2.7141 -174.8455>}
  {
   shapeType=Capsule
   length=5.25
   radius=36.25
   shapeName=attach_capsule
   center=<-4.0027 -8.996 1.0408>
   rotation=<rot -90 179.9993 -179.9995>
  }
 ]
 spine_01 [{shapeType=Capsule, length=11.952, radius=16.3909, center=<-4.4394 -3.528 0>}]
 spine_02 [{shapeType=Capsule, length=12.1707, radius=15.8284, center=<-4.4394 -3.7915 0.0001>}]
 spine_03 [{shapeType=Capsule, length=12.7332, radius=13.8284, center=<-4.4394 -3.6514 0>}]
 thigh_l [{shapeType=Capsule, length=23.2194, radius=14.4465, center=<-19.9578 -1.7916 0.0178>, rotation=<rot 33.5417 94.5875 92.8247>}]
 thigh_r [{shapeType=Capsule, length=23.2194, radius=14.4465, center=<19.9578 1.7916 -0.0178>, rotation=<rot 33.5417 94.5875 -87.1753>}]
}
constraints [
 {bone1Name=jacket_left02, bone2Name=jacket_left01, swing1LimitDegrees=45, swing2LimitDegrees=45, twistLimitDegrees=5}
 {bone1Name=jacket_left03, bone2Name=jacket_left02, swing1LimitDegrees=45, swing2LimitDegrees=45, twistLimitDegrees=5}
 {bone1Name=jacket_left05, bone2Name=jacket_left04, swing1LimitDegrees=45, swing2LimitDegrees=45, twistLimitDegrees=5}
 {bone1Name=jacket_left06, bone2Name=jacket_left05, swing1LimitDegrees=45, swing2LimitDegrees=45, twistLimitDegrees=5}
 {bone1Name=jacket_back_left02, bone2Name=jacket_back_left01, swing1LimitDegrees=45, swing2LimitDegrees=45, twistLimitDegrees=5}
 {bone1Name=jacket_back_left03, bone2Name=jacket_back_left02, swing1LimitDegrees=45, swing2LimitDegrees=45, twistLimitDegrees=5}
 {bone1Name=jacket_back_left04, bone2Name=jacket_back_left03, swing1LimitDegrees=45, swing2LimitDegrees=45, twistLimitDegrees=5}
 {bone1Name=jacket_back_right02, bone2Name=jacket_back_right01, swing1LimitDegrees=45, swing2LimitDegrees=45, twistLimitDegrees=5}
 {bone1Name=jacket_back_right03, bone2Name=jacket_back_right02, swing1LimitDegrees=45, swing2LimitDegrees=45, twistLimitDegrees=5}
 {bone1Name=jacket_back_right04, bone2Name=jacket_back_right03, swing1LimitDegrees=45, swing2LimitDegrees=45, twistLimitDegrees=5}
 {bone1Name=jacket_front_left01, bone2Name=jacket_front_top_driver_l, swing1LimitDegrees=15, swing2LimitDegrees=30, twistLimitDegrees=30}
 {bone1Name=jacket_front_left02, bone2Name=jacket_front_left01, swing1LimitDegrees=35, swing2LimitDegrees=30, twistLimitDegrees=30}
 {bone1Name=jacket_front_left0103, bone2Name=jacket_front_left02, swing1LimitDegrees=35, swing2LimitDegrees=30, twistLimitDegrees=30}
 {bone1Name=jacket_front_left_dist, bone2Name=jacket_front_left0103, swing1LimitDegrees=35, swing2LimitDegrees=30, twistLimitDegrees=30}
 {bone1Name=jacket_front_right01, bone2Name=jacket_front_top_driver_r, swing1LimitDegrees=15, swing2LimitDegrees=30, twistLimitDegrees=30}
 {bone1Name=jacket_front_right02, bone2Name=jacket_front_right01, swing1LimitDegrees=35, swing2LimitDegrees=30, twistLimitDegrees=30}
 {bone1Name=jacket_front_right03, bone2Name=jacket_front_right02, swing1LimitDegrees=35, swing2LimitDegrees=30, twistLimitDegrees=30}
 {bone1Name=jacket_front_right_dist, bone2Name=jacket_front_right03, swing1LimitDegrees=35, swing2LimitDegrees=30, twistLimitDegrees=30}
 {
  bone1Name=thigh_l
  bone2Name=pelvis
  swing1LimitDegrees=45
  swing1Motion=Free
  swing2LimitDegrees=45
  swing2Motion=Free
  twistLimitDegrees=45
  twistMotion=Free
 }
 {bone1Name=jacket_left01, bone2Name=jacket_left01_MSTR, swing1LimitDegrees=45, swing2LimitDegrees=45, twistLimitDegrees=5}
 {bone1Name=jacket_left04, bone2Name=jacket_left04_MSTR, swing1LimitDegrees=45, swing2LimitDegrees=45, twistLimitDegrees=5}
 {bone1Name=jacket_back_left01, bone2Name=pelvis, swing1LimitDegrees=45, swing2LimitDegrees=45, twistLimitDegrees=5}
 {bone1Name=jacket_back_right01, bone2Name=pelvis, swing1LimitDegrees=45, swing2LimitDegrees=45, twistLimitDegrees=5}
 {
  bone1Name=jacket_front_top_driver_l
  bone2Name=jacket_front_top_driver_l_MSTR
  swing1LimitDegrees=15
  swing2LimitDegrees=30
  twistLimitDegrees=30
 }
 {
  bone1Name=jacket_front_top_driver_r
  bone2Name=jacket_front_top_driver_r_MSTR
  swing1LimitDegrees=15
  swing2LimitDegrees=30
  twistLimitDegrees=30
 }
 {
  bone1Name=calf_l
  bone2Name=thigh_l
  swing1LimitDegrees=45
  swing1Motion=Free
  swing2LimitDegrees=45
  swing2Motion=Free
  twistLimitDegrees=45
  twistMotion=Free
 }
 {
  bone1Name=calf_r
  bone2Name=thigh_r
  swing1LimitDegrees=45
  swing1Motion=Free
  swing2LimitDegrees=45
  swing2Motion=Free
  twistLimitDegrees=45
  twistMotion=Free
 }
 {
  bone1Name=jacket_back_right04
  bone2Name=jacket_back_left04
  swing1LimitDegrees=45
  swing1Motion=Free
  swing2LimitDegrees=45
  swing2Motion=Free
  twistLimitDegrees=45
  twistMotion=Free
 }
 {
  bone1Name=jacket_front_top_driver_l_MSTR
  bone2Name=spine_03
  swing1LimitDegrees=45
  swing1Motion=Free
  swing2LimitDegrees=45
  swing2Motion=Free
  twistLimitDegrees=45
  twistMotion=Free
 }
 {
  bone1Name=jacket_front_top_driver_r_MSTR
  bone2Name=spine_03
  swing1LimitDegrees=45
  swing1Motion=Free
  swing2LimitDegrees=45
  swing2Motion=Free
  twistLimitDegrees=45
  twistMotion=Free
 }
 {
  bone1Name=spine_03
  bone2Name=spine_02
  swing1LimitDegrees=45
  swing1Motion=Free
  swing2LimitDegrees=45
  swing2Motion=Free
  twistLimitDegrees=45
  twistMotion=Free
 }
 {
  bone1Name=spine_02
  bone2Name=spine_01
  swing1LimitDegrees=45
  swing1Motion=Free
  swing2LimitDegrees=45
  swing2Motion=Free
  twistLimitDegrees=45
  twistMotion=Free
 }
 {
  bone1Name=spine_01
  bone2Name=pelvis
  swing1LimitDegrees=45
  swing1Motion=Free
  swing2LimitDegrees=45
  swing2Motion=Free
  twistLimitDegrees=45
  twistMotion=Free
 }
 {
  bone1Name=jacket_left02
  bone2Name=jacket_front_left0103
  swing1LimitDegrees=45
  swing1Motion=Free
  swing2LimitDegrees=45
  swing2Motion=Free
  twistLimitDegrees=45
  twistMotion=Free
 }
 {
  bone1Name=jacket_left05
  bone2Name=jacket_front_right03
  swing1LimitDegrees=45
  swing1Motion=Free
  swing2LimitDegrees=45
  swing2Motion=Free
  twistLimitDegrees=45
  twistMotion=Free
 }
 {
  bone1Name=jacket_left04_MSTR
  bone2Name=pelvis
  swing1LimitDegrees=45
  swing1Motion=Free
  swing2LimitDegrees=45
  swing2Motion=Free
  twistLimitDegrees=45
  twistMotion=Free
 }
 {
  bone1Name=jacket_left01_MSTR
  bone2Name=pelvis
  swing1LimitDegrees=45
  swing1Motion=Free
  swing2LimitDegrees=45
  swing2Motion=Free
  twistLimitDegrees=45
  twistMotion=Free
 }
 {
  bone1Name=thigh_r
  bone2Name=pelvis
  swing1LimitDegrees=45
  swing1Motion=Free
  swing2LimitDegrees=45
  swing2Motion=Free
  twistLimitDegrees=45
  twistMotion=Free
 }
 {
  bone1Name=foot_r
  bone2Name=calf_r
  swing1LimitDegrees=45
  swing1Motion=Free
  swing2LimitDegrees=45
  swing2Motion=Free
  twistLimitDegrees=45
  twistMotion=Free
 }
 {
  bone1Name=foot_l
  bone2Name=calf_l
  swing1LimitDegrees=45
  swing1Motion=Free
  swing2LimitDegrees=45
  swing2Motion=Free
  twistLimitDegrees=45
  twistMotion=Free
 }
 {
  bone1Name=pelvis
  bone2Name=root
  swing1LimitDegrees=45
  swing1Motion=Free
  swing2LimitDegrees=45
  swing2Motion=Free
  twistLimitDegrees=45
  twistMotion=Free
 }
]
