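# Post-process style AnimBP for the M_MED_SRDrift tier 05-06 player body, adding secondary motion to jacket, sleeves, collar and hair.
# It copies the parent mesh pose, then layers elbow/head Pose Drivers, AnimDynamics and leg- and crouch-weighted bone offsets into a 'Dynamics' cache.
# A bool blend picks RigidBody jacket simulation, with standard and alternate gravity, or thigh constraints.
# The output optionally shrinks the head for custom character parts.
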AnimBlueprint /BRCosmetics/Characters/Player/Male/Medium/Bodies/M_MED_SRDrift/Meshes/M_MED_SRDrift_tier0506_AnimBP
target_skeleton=&/M_MED_SRDrift_Skeleton
graphs [
 AnimGraph {nodes [
   Anim Dynamics - Bone: dyn_cube_r_in_02(Alpha=0.4, ExternalForce="0,0,0") {
    angularDampingOverride=0.9
    angularSpringConstant=200000
    bAngularSpring=true
    bGravityOverrideInSimSpace=true
    bLinearSpring=true
    bOverrideAngularDamping=true
    bOverrideLinearDamping=true
    gravityScale=0
    linearDampingOverride=0.995
    linearSpringConstant=50
    simulationSpace=BoneRelative
    alphaBoolBlend {blendOption=Linear}
    boundBone {boneName=dyn_cube_r_in_02}
    componentAppliedLinearAccClamp=<6000 6000 6000>
    componentLinearAccScale=<0.3 0.3 0.1>
    componentLinearVelScale=<0.7 0.7 0.8>
    gravityOverride {z=-980}
    physicsBodyDefinitions [{
      sphereCollisionRadius=0
      boundBone {boneName=dyn_cube_r_in_02}
      boxExtents=<5 5 5>
      constraintSetup {
       angularTargetAxis=AxisZ
       twistAxis=AxisZ
       angularLimitsMax=<0 0 0>
       angularLimitsMin=<0 0 0>
       angularTarget=<0 0 0>
       linearAxesMax=<18 18 18>
       linearAxesMin=<-0.1 -0.1 -2>
      }
      localJointOffset=<-0.1 -0.2 0>
     }]
    relativeSpaceBone {boneName=dyn_master}
   }
   n1 = Get DiveJiggleRotation() -> (DiveJiggleRotation:Rotator)
   n2 = "Transform (Modify) Bone - Bone: jacket_back_left02"(Alpha=0.7, Translation="0,0,0", Rotation=@n1, Scale="1,1,1") {rotationMode=BMM_Additive, rotationSpace=BCS_BoneSpace, alphaBoolBlend {blendOption=Linear}, boneToModify {boneName=jacket_back_left02}}
   n3 = "Transform (Modify) Bone - Bone: jacket_back_right02"(ComponentPose=@n2, Alpha=0.7, Translation="0,0,0", Rotation=@n1, Scale="1,1,1") {rotationMode=BMM_Additive, rotationSpace=BCS_BoneSpace, alphaBoolBlend {blendOption=Linear}, boneToModify {boneName=jacket_back_right02}}
   n4 = "Transform (Modify) Bone - Bone: jacket_left04"(ComponentPose=@n3, Alpha=0.7, Translation="0,0,0", Rotation=@n1, Scale="1,1,1") {rotationMode=BMM_Additive, rotationSpace=BCS_BoneSpace, alphaBoolBlend {blendOption=Linear}, boneToModify {boneName=jacket_left04}}
   n5 = "Transform (Modify) Bone - Bone: jacket_left01"(ComponentPose=@n4, Alpha=0.7, Translation="0,0,0", Rotation=@n1, Scale="1,1,1") {rotationMode=BMM_Additive, rotationSpace=BCS_BoneSpace, alphaBoolBlend {blendOption=Linear}, boneToModify {boneName=jacket_left01}}
   n6 = "Transform (Modify) Bone - Bone: jacket_left04"(ComponentPose=@n5, Alpha=0.7, Translation="0,0,0", Rotation="0,0,0", Scale="1,1,1") {rotationMode=BMM_Additive, rotationSpace=BCS_BoneSpace, alphaBoolBlend {blendOption=Linear}, boneToModify {boneName=jacket_left04}}
   n7 = "Transform (Modify) Bone - Bone: jacket_left01"(ComponentPose=@n6, Alpha=0.7, Translation="0,0,0", Rotation="0,0,0", Scale="1,1,1") {rotationMode=BMM_Additive, rotationSpace=BCS_BoneSpace, alphaBoolBlend {blendOption=Linear}, boneToModify {boneName=jacket_left01}}
   n8 = Get JacketTailGroundAlpha() -> (JacketTailGroundAlpha:float)
   n9 = M_MED_SRDrift_elbows_pose : left_knee_bend!(PoseWeight:float=@n8) -> (Pose:Pose Link Structure) {
    is=title AnimGraphNode_PoseByName
    poseAsset=/BRCosmetics/Characters/Player/Male/Medium/Bodies/M_MED_SRDrift/Meshes/pose_assets/M_MED_SRDrift_elbows_pose
    poseName=left_knee_bend
   }
   n10 = Get bIsCrouching()
   n11 = Apply Additive!(Base=Component To Local(ComponentPose=@n7), Additive=@n9, bAlphaBoolEnabled:bool=@n10) -> (Pose) {alphaInputType=Bool, alphaBoolBlend {blendOption=Linear}}
   n12 = M_MED_SRDrift_elbows_pose : right_knee_bend!(PoseWeight:float=@n8) -> (Pose:Pose Link Structure) {
    is=title AnimGraphNode_PoseByName
    poseAsset=&M_MED_SRDrift_elbows_pose
    poseName=right_knee_bend
   }
   n13 = M_MED_SRDrift_elbows_pose : right_knee_bend(PoseWeight=Get RightFootJacketBackDistanceAlpha() -> (RightFootJacketBackDistanceAlpha:float)) {
    poseAsset=&M_MED_SRDrift_elbows_pose
    poseName=right_knee_bend
   }
   n14 = M_MED_SRDrift_elbows_pose : left_knee_bend(PoseWeight=Get LeftFootJacketBackDistanceAlpha() -> (LeftFootJacketBackDistanceAlpha:float)) {
    poseAsset=&M_MED_SRDrift_elbows_pose
    poseName=left_knee_bend
   }
   Apply Additive!(
    Base=Apply Additive(
     Base=Apply Additive!(Base=@n11, Additive=@n12, bAlphaBoolEnabled:bool=@n10) -> (Pose) {alphaInputType=Bool, alphaBoolBlend {blendOption=Linear}}
     Additive=@n13
     Alpha=0.6
    ) {alphaBoolBlend {blendOption=Linear}}
    Additive=@n14
    Alpha:float=0.6
   ) -> (Pose) {alphaBoolBlend {blendOption=Linear}}
   n15 = Get LegJacketFrontAlpha() -> (LegJacketFrontAlpha:float)
   n16 = M_MED_SRDrift_elbows_pose : right_leg_fwd!(PoseWeight:float=@n15) -> (Pose:Pose Link Structure) {
    is=title AnimGraphNode_PoseByName
    poseAsset=&M_MED_SRDrift_elbows_pose
    poseName=right_leg_fwd
   }
   n17 = M_MED_SRDrift_elbows_pose : left_leg_fwd!(PoseWeight:float=@n15) -> (Pose:Pose Link Structure) {
    is=title AnimGraphNode_PoseByName
    poseAsset=&M_MED_SRDrift_elbows_pose
    poseName=left_leg_fwd
   }
   Apply Additive!(
    Base=Apply Additive(Additive=@n16, Alpha=0.8) {alphaBoolBlend {blendOption=Linear}, alphaScaleBiasClamp {interpSpeedDecreasing=1, interpSpeedIncreasing=1}}
    Additive=@n17
    Alpha:float=0.8
   ) -> (Pose) {alphaBoolBlend {blendOption=Linear}}
   "Constraint - dyn_hair_01 (P)"(Alpha=0.35, ConstraintWeights_0=1) {
    alphaBoolBlend {blendOption=Linear}
    alphaScaleBiasClamp {bClampResult=true, clampMax=0.35}
    boneToModify {boneName=dyn_hair_01}
    constraintSetup [{targetBone {boneName=clavicle_l}}]
    constraintWeights [1]
   }
   n18 = Get FallingAnimDynamicsUpForce() -> (FallingAnimDynamicsUpForce:Vector)
   n19 = Get AnimDynamicsAlpha() -> (AnimDynamicsAlpha:float)
   n20 = "Anim Dynamics - Chain: jacket_back_left02 -> jacket_back_left04"(Alpha=@n19, ExternalForce=@n18) {
    angularDampingOverride=1
    angularSpringConstant=500
    bAngularSpring=true
    bChain=true
    bGravityOverrideInSimSpace=true
    bOverrideAngularDamping=true
    bUseGravityOverride=true
    numSolverIterationsPostUpdate=4
    numSolverIterationsPreUpdate=16
    simulationSpace=RootRelative
    alphaBoolBlend {blendOption=Linear}
    boundBone {boneName=jacket_back_left02}
    chainEnd {boneName=jacket_back_left04}
    componentAppliedLinearAccClamp=<3000 3000 3000>
    componentLinearAccScale=<0.4 0.3 0.3>
    componentLinearVelScale=<0.4 0.3 0.3>
    gravityOverride {z=-980}
    physicsBodyDefinitions [{
      sphereCollisionRadius=0
      boundBone {boneName=jacket_back_left02}
      boxExtents=<10 20 20>
      constraintSetup {angularLimitsMax=<5 80 80>, angularLimitsMin=<-5 -80 -80>, angularTarget=<1 0 0>}
      localJointOffset=<2 0 0>
     }]
   }
   n21 = "Anim Dynamics - Chain: jacket_back_right02 -> jacket_back_right04"(ComponentPose=@n20, Alpha=Get CrouchJacketBackRightAlpha() -> (CrouchJacketBackRightAlpha:float), ExternalForce=@n18) {
    angularDampingOverride=1
    angularSpringConstant=500
    bAngularSpring=true
    bChain=true
    bGravityOverrideInSimSpace=true
    bOverrideAngularDamping=true
    bUseGravityOverride=true
    numSolverIterationsPostUpdate=4
    numSolverIterationsPreUpdate=16
    simulationSpace=RootRelative
    alphaBoolBlend {blendOption=Linear}
    boundBone {boneName=jacket_back_right02}
    chainEnd {boneName=jacket_back_right04}
    componentAppliedLinearAccClamp=<3000 3000 3000>
    componentLinearAccScale=<0.4 0.3 0.3>
    componentLinearVelScale=<0.4 0.3 0.3>
    gravityOverride {z=-980}
    physicsBodyDefinitions [{
      sphereCollisionRadius=0
      boundBone {boneName=jacket_back_right02}
      boxExtents=<10 20 20>
      constraintSetup {angularLimitsMax=<5 80 80>, angularLimitsMin=<-5 -80 -80>, angularTarget=<1 0 0>}
      localJointOffset=<-2 0 0>
     }]
   }
   n22 = "Anim Dynamics - Chain: jacket_left01 -> jacket_left03"(ComponentPose=@n21, Alpha=@n19, ExternalForce=@n18) {
    angularDampingOverride=1
    angularSpringConstant=500
    bAngularSpring=true
    bChain=true
    bGravityOverrideInSimSpace=true
    bOverrideAngularDamping=true
    bUseGravityOverride=true
    numSolverIterationsPostUpdate=4
    numSolverIterationsPreUpdate=16
    simulationSpace=RootRelative
    alphaBoolBlend {blendOption=Linear}
    boundBone {boneName=jacket_left01}
    chainEnd {boneName=jacket_left03}
    componentAppliedLinearAccClamp=<3000 3000 3000>
    componentLinearAccScale=<0.4 0.3 0.3>
    componentLinearVelScale=<0.4 0.3 0.3>
    gravityOverride {z=-980}
    physicsBodyDefinitions [{
      sphereCollisionRadius=0
      boundBone {boneName=jacket_left01}
      boxExtents=<10 20 20>
      constraintSetup {angularLimitsMax=<5 60 20>, angularLimitsMin=<-5 -60 -20>, angularTarget=<1 0 0>}
      localJointOffset=<-2 0 0>
     }]
    planarLimits [{drivingBone {boneName=thigh_l}, planeTransform {location=<0 0 0>, rotation=<rot 0 0 -180>}}]
   }
   n23 = "Anim Dynamics - Chain: jacket_left04 -> jacket_left06"(ComponentPose=@n22, Alpha=@n19, ExternalForce=@n18) {
    angularDampingOverride=1
    angularSpringConstant=500
    bAngularSpring=true
    bChain=true
    bGravityOverrideInSimSpace=true
    bOverrideAngularDamping=true
    bUseGravityOverride=true
    numSolverIterationsPostUpdate=4
    numSolverIterationsPreUpdate=16
    simulationSpace=RootRelative
    alphaBoolBlend {blendOption=Linear}
    boundBone {boneName=jacket_left04}
    chainEnd {boneName=jacket_left06}
    componentAppliedLinearAccClamp=<3000 3000 3000>
    componentLinearAccScale=<0.4 0.3 0.3>
    componentLinearVelScale=<0.4 0.3 0.3>
    gravityOverride {z=-980}
    physicsBodyDefinitions [{
      sphereCollisionRadius=0
      boundBone {boneName=jacket_left04}
      boxExtents=<10 20 20>
      constraintSetup {angularLimitsMax=<5 60 20>, angularLimitsMin=<-5 -60 -20>, angularTarget=<1 0 0>}
      localJointOffset=<2 0 0>
     }]
   }
   n24 = "Anim Dynamics - Chain: jacket_front_left01 -> jacket_front_left_dist"(ComponentPose=@n23, Alpha=Get CrouchJacketFrontLeftAlpha() -> (CrouchJacketFrontLeftAlpha:float), ExternalForce=@n18) {
    angularDampingOverride=1
    angularSpringConstant=500
    bAngularSpring=true
    bChain=true
    bGravityOverrideInSimSpace=true
    bOverrideAngularDamping=true
    bUseGravityOverride=true
    numSolverIterationsPostUpdate=10
    numSolverIterationsPreUpdate=16
    simulationSpace=RootRelative
    alphaBoolBlend {blendOption=Linear}
    boundBone {boneName=jacket_front_left01}
    chainEnd {boneName=jacket_front_left_dist}
    componentAppliedLinearAccClamp=<3000 3000 3000>
    componentLinearAccScale=<0.4 0.3 0.3>
    componentLinearVelScale=<0.4 0.3 0.3>
    gravityOverride {z=-980}
    physicsBodyDefinitions [{
      collisionType=CustomSphere
      sphereCollisionRadius=3
      boundBone {boneName=jacket_front_left01}
      boxExtents=<10 20 20>
      constraintSetup {angularLimitsMax=<5 50 3>, angularLimitsMin=<-5 -50 -3>, angularTarget=<1 0 0>}
      localJointOffset=<-2 0 0>
     }]
    planarLimits [{drivingBone {boneName=thigh_l}, planeTransform {location=<0 12 0>, rotation=<rot 0 -5 90>}}]
   }
   n25 = "Anim Dynamics - Chain: jacket_front_right01 -> jacket_front_right_dist"(ComponentPose=@n24, Alpha=Get CrouchJacketFrontRightAlpha() -> (CrouchJacketFrontRightAlpha:float), ExternalForce=@n18) {
    angularDampingOverride=1
    angularSpringConstant=500
    bAngularSpring=true
    bChain=true
    bGravityOverrideInSimSpace=true
    bOverrideAngularDamping=true
    bUseGravityOverride=true
    numSolverIterationsPostUpdate=4
    numSolverIterationsPreUpdate=16
    simulationSpace=RootRelative
    alphaBoolBlend {blendOption=Linear}
    boundBone {boneName=jacket_front_right01}
    chainEnd {boneName=jacket_front_right_dist}
    componentAppliedLinearAccClamp=<3000 3000 3000>
    componentLinearAccScale=<0.4 0.3 0.3>
    componentLinearVelScale=<0.4 0.3 0.3>
    gravityOverride {z=-980}
    physicsBodyDefinitions [{
      collisionType=CustomSphere
      sphereCollisionRadius=3
      boundBone {boneName=jacket_front_right01}
      boxExtents=<10 20 20>
      constraintSetup {angularLimitsMax=<5 50 3>, angularLimitsMin=<-5 -50 -3>, angularTarget=<1 0 0>}
      localJointOffset=<2 0 0>
     }]
    planarLimits [{drivingBone {boneName=thigh_r}, planeTransform {location=<0 -12 0>, rotation=<rot 0 -5 -90>}}]
   }
   n26 = Copy Bone - Source Bone: jacket_front_left01 - Target Bone: jacket_front_top_driver_l!(ComponentPose:Component Space Pose Link Structure=@n25, Alpha:float=Get JacketCopyBoneAlpha() -> (JacketCopyBoneAlpha:float)) -> (Pose:Component Space Pose Link Structure) {
    is=title AnimGraphNode_CopyBone
    bCopyRotation=true
    controlSpace=BCS_ParentBoneSpace
    alphaBoolBlend {blendOption=Linear}
    sourceBone {boneName=jacket_front_left01}
    targetBone {boneName=jacket_front_top_driver_l}
   }
   Copy Bone - Source Bone: jacket_front_right01 - Target Bone: jacket_front_top_driver_r!(ComponentPose:Component Space Pose Link Structure=@n26, Alpha:float=0.2) -> (Pose:Component Space Pose Link Structure) {
    is=title AnimGraphNode_CopyBone
    bCopyRotation=true
    alphaBoolBlend {blendOption=Linear}
    sourceBone {boneName=jacket_front_right01}
    targetBone {boneName=jacket_front_top_driver_r}
   }
   n27 = Pose Driver - Source: thigh_l!(SourcePose:Pose Link Structure) -> (Pose:Pose Link Structure) {
    is=title AnimGraphNode_PoseDriver
    poseAsset=&M_MED_SRDrift_elbows_pose
    onlyDriveBones [
     {boneName=jacket_back_right03}
     {boneName=jacket_back_right04}
     {boneName=jacket_front_left01}
     {boneName=jacket_front_left02}
     {boneName=jacket_front_left0103}
     {boneName=jacket_front_left_dist}
     {boneName=jacket_front_top_driver_l}
    ]
    poseTargets [{
      bApplyCustomCurve=false
      bIsHidden=false
      distanceMethod=DefaultMethod
      drivenName=crouch_pose
      functionType=DefaultFunction
      targetScale=1
      boneTransforms [
       {targetRotation=<rot 26.0452 -106.7931 -5.0033>, targetTranslation=<-2.5693 -0.0109 -7.7501>}
       {targetRotation=<rot 19.3215 151.5046 -15.782>, targetTranslation=<-2.5691 -0.0109 7.7501>}
       {targetRotation=<rot 0.0722 114.6483 0.5522>, targetTranslation=<-38.8265 0.3215 0.0017>}
       {targetRotation=<rot 0.1962 118.9635 0.4243>, targetTranslation=<38.8358 -0.3388 -0.0016>}
       {targetRotation=<rot -4.9269 -9.8929 -12.3042>, targetTranslation=<-42.561 0.1077 0.4292>}
       {targetRotation=<rot 9.1332 21.85 -60.8857>, targetTranslation=<42.5612 -0.1076 -0.4293>}
      ]
      customCurve {defaultValue=3.4028234663852886e+38, postInfinityExtrap=RCCE_Constant, preInfinityExtrap=RCCE_Constant, keys []}
      targetRotation=<rot 0 0 0>
     }]
    rBFParams {radius=96.2714}
    sourceBones [{boneName=thigh_l}, {boneName=thigh_r}, {boneName=calf_l}, {boneName=calf_r}, {boneName=foot_l}, {boneName=foot_r}]
   }
   Local To Component(LocalPose=@n27)
   n28 = "Transform (Modify) Bone - Bone: jacket_front_top_driver_l"(Alpha=Get legUp_L() -> (legUp_L:double), Translation="0,-10,0", Rotation="120,0,0.000", Scale="1,1,1") {
    rotationMode=BMM_Additive
    rotationSpace=BCS_BoneSpace
    translationMode=BMM_Additive
    translationSpace=BCS_BoneSpace
    alphaBoolBlend {blendOption=Linear}
    boneToModify {boneName=jacket_front_top_driver_l}
    rotation {pitch=120}
    translation {y=-10}
   }
   n29 = "Transform (Modify) Bone - Bone: jacket_front_top_driver_r"(ComponentPose=@n28, Alpha=Get legUp_R() -> (legUp_R:double), Translation="0,10,0", Rotation="130,0,0", Scale="1,1,1") {
    rotationMode=BMM_Additive
    rotationSpace=BCS_BoneSpace
    translationMode=BMM_Additive
    translationSpace=BCS_BoneSpace
    alphaBoolBlend {blendOption=Linear}
    boneToModify {boneName=jacket_front_top_driver_r}
    rotation {pitch=130}
    translation {y=10}
   }
   n30 = "Transform (Modify) Bone - Bone: jacket_front_top_driver_r"(ComponentPose=@n29, Alpha=Get legOut_R() -> (legOut_R:double), Translation="0,0,8", Rotation="20,90,0", Scale="1,1,1") {
    rotationMode=BMM_Additive
    rotationSpace=BCS_BoneSpace
    translationMode=BMM_Additive
    translationSpace=BCS_BoneSpace
    alphaBoolBlend {blendOption=Linear}
    boneToModify {boneName=jacket_front_top_driver_r}
    rotation {pitch=20, yaw=90}
    translation {z=8}
   }
   "Transform (Modify) Bone - Bone: jacket_front_top_driver_l"(ComponentPose=@n30, Alpha=Get legOut_L() -> (legOut_L:double), Translation="10,0,-15", Rotation="-15,90,0", Scale="1,1,1") {
    rotationMode=BMM_Additive
    rotationSpace=BCS_BoneSpace
    translationMode=BMM_Additive
    translationSpace=BCS_BoneSpace
    alphaBoolBlend {blendOption=Linear}
    boneToModify {boneName=jacket_front_top_driver_l}
    rotation {pitch=-15, yaw=90}
    translation {x=10, z=-15}
   }
   "Transform (Modify) Bone - Bone: dyn_cube_l_out_04"(Alpha=0.8, Translation="0,0,0", Rotation="0,0,0", Scale="1,1,1") {
    rotationMode=BMM_Additive
    rotationSpace=BCS_BoneSpace
    scaleMode=BMM_Additive
    scaleSpace=BCS_BoneSpace
    translationMode=BMM_Additive
    translationSpace=BCS_BoneSpace
    alphaBoolBlend {blendOption=Linear}
    boneToModify {boneName=dyn_cube_l_out_04}
   }
   n31 = "Transform (Modify) Bone - Bone: jacket_front_top_driver_l"(Alpha=1, Translation="9.773,-3.436,1.672", Rotation="53.851,67.750,68.552", Scale="1,1,1") {
    rotationMode=BMM_Additive
    rotationSpace=BCS_BoneSpace
    translationMode=BMM_Additive
    translationSpace=BCS_BoneSpace
    alphaBoolBlend {blendOption=Linear}
    alphaScaleBiasClamp {bInterpResult=true, interpSpeedDecreasing=15, interpSpeedIncreasing=15}
    boneToModify {boneName=jacket_front_top_driver_l}
    rotation=<rot 53.8508 67.7503 68.5522>
    translation=<9.773 -3.436 1.672>
   }
   "Transform (Modify) Bone - Bone: jacket_front_top_driver_r"(ComponentPose=@n31, Alpha=1, Translation="-1.905,4.595,-3.231", Rotation="49.100,86.919,85.926", Scale="1,1,1") {
    rotationMode=BMM_Additive
    rotationSpace=BCS_BoneSpace
    translationMode=BMM_Additive
    translationSpace=BCS_BoneSpace
    alphaBoolBlend {blendOption=Linear}
    alphaScaleBiasClamp {bInterpResult=true, interpSpeedDecreasing=15, interpSpeedIncreasing=15}
    boneToModify {boneName=jacket_front_top_driver_r}
    rotation=<rot 49.1 86.919 85.926>
    translation=<-1.905 4.595 -3.231>
   }
   "Transform (Modify) Bone - Bone: jacket_front_top_driver_l_MSTR"(Alpha=Get legUp_L(), Translation="-6.216,-5.872,1.108", Rotation="-21.928,35.930,51.234", Scale="1,1,1") {
    rotationMode=BMM_Additive
    rotationSpace=BCS_BoneSpace
    translationMode=BMM_Additive
    translationSpace=BCS_BoneSpace
    alphaBoolBlend {blendOption=Linear}
    alphaScaleBiasClamp {bInterpResult=true}
    boneToModify {boneName=jacket_front_top_driver_l_MSTR}
    rotation=<rot -21.928 35.93 51.2343>
    translation=<-6.216 -5.872 1.108>
   }
   "Transform (Modify) Bone - Bone: jacket_collar_back"!(ComponentPose, Alpha=0.8, Translation="-0.2,-2.5,0") -> (Pose) {
    rotationMode=BMM_Additive
    rotationSpace=BCS_BoneSpace
    scaleMode=BMM_Additive
    scaleSpace=BCS_BoneSpace
    translationMode=BMM_Additive
    translationSpace=BCS_BoneSpace
    alphaBoolBlend {blendOption=Linear}
    boneToModify {boneName=jacket_collar_back}
    translation=<-0.2 -2.5>
   }
   n32 = "Transform (Modify) Bone - Bone: head"!(ComponentPose=Local To Component(LocalPose=Use cached pose 'PreScale Cache'()), Alpha=1) -> (Pose) {scaleMode=BMM_Replace, alphaBoolBlend {blendOption=Linear}, boneToModify {boneName=head}, scale=<0.01 0.01 0.01>}
   Output Pose(Result=Blend Poses by bool(
     BlendPose_0=Component To Local(ComponentPose=@n32)
     BlendPose_1=Use cached pose 'PreScale Cache'()
     BlendTime_0=0
     BlendTime_1=0
     bActiveValue=Get GetPlayerListUsesCustomCharacterParts() -> (GetPlayerListUsesCustomCharacterParts:bool)
    ) {blendType=Linear}) {layerGroup=None}
   n33 = Pose Driver - Source: lowerarm_l!(SourcePose:Pose Link Structure=Copy Pose From Mesh() {bUseAttachedParent=true}) -> (Pose:Pose Link Structure) {
    is=title AnimGraphNode_PoseDriver
    poseAsset=&M_MED_SRDrift_elbows_pose
    onlyDriveBones [{boneName=jacket_sleeve_driven_l}]
    poseTargets [{
      bApplyCustomCurve=false
      bIsHidden=false
      distanceMethod=DefaultMethod
      drivenName=left_elbow_bend
      functionType=DefaultFunction
      targetScale=1
      boneTransforms [{targetRotation=<rot -3.9746 120.6689 -2.3538>, targetTranslation=<26.6514 0 0>}]
      customCurve {defaultValue=3.4028234663852886e+38, postInfinityExtrap=RCCE_Constant, preInfinityExtrap=RCCE_Constant, keys []}
      targetRotation=<rot 0 0 0>
     }]
    rBFParams {distanceMethod=Euclidean, function=Linear, radius=46.2842}
    sourceBones [{boneName=lowerarm_l}]
   }
   n34 = Pose Driver - Source: lowerarm_r!(SourcePose:Pose Link Structure=@n33) -> (Pose:Pose Link Structure) {
    is=title AnimGraphNode_PoseDriver
    poseAsset=&M_MED_SRDrift_elbows_pose
    onlyDriveBones [{boneName=jacket_sleeve_driven_r}]
    poseTargets [{
      bApplyCustomCurve=false
      bIsHidden=false
      distanceMethod=DefaultMethod
      drivenName=right_elbow_bend
      functionType=DefaultFunction
      targetScale=1
      boneTransforms [{targetRotation=<rot -2.196 120 4.0638>, targetTranslation=<-26.6515 0 0>}]
      customCurve {defaultValue=3.4028234663852886e+38, postInfinityExtrap=RCCE_Constant, preInfinityExtrap=RCCE_Constant, keys []}
      targetRotation=<rot 0 0 0>
     }]
    rBFParams {distanceMethod=Euclidean, function=Linear, radius=45.8299}
    sourceBones [{boneName=lowerarm_r}]
   }
   n35 = Anim Dynamics - Bone: jacket_sleeve_l(
    ComponentPose=Local To Component(LocalPose=@n34)
    Alpha=Get JacketLeftSleeveDynamicAlpha() -> (JacketLeftSleeveDynamicAlpha:float)
    ExternalForce=Get WeaponKickupExternalForceLeft() -> (WeaponKickupExternalForceLeft:Vector)
   ) {
    angularDampingOverride=0.99
    angularSpringConstant=5000
    bAngularSpring=true
    bGravityOverrideInSimSpace=true
    bOverrideAngularDamping=true
    numSolverIterationsPostUpdate=4
    numSolverIterationsPreUpdate=16
    simulationSpace=Actor
    alphaBoolBlend {blendOption=Linear}
    boundBone {boneName=jacket_sleeve_l}
    physicsBodyDefinitions [{
      sphereCollisionRadius=0
      boundBone {boneName=jacket_sleeve_l}
      boxExtents=<10 20 20>
      constraintSetup {angularLimitsMax=<5 20 10>, angularLimitsMin=<-5 -5 -10>, angularTarget=<1 0 0>}
      localJointOffset=<3 0 0>
     }]
   }
   n36 = Get WeaponKickupExternalForce() -> (WeaponKickupExternalForce:Vector)
   n37 = Anim Dynamics - Bone: jacket_sleeve_r(ComponentPose=@n35, Alpha=Get JacketRightSleeveDynamicAlpha() -> (JacketRightSleeveDynamicAlpha:float), ExternalForce=@n36) {
    angularDampingOverride=0.99
    angularSpringConstant=5000
    bAngularSpring=true
    bGravityOverrideInSimSpace=true
    bOverrideAngularDamping=true
    numSolverIterationsPostUpdate=4
    numSolverIterationsPreUpdate=16
    simulationSpace=Actor
    alphaBoolBlend {blendOption=Linear}
    boundBone {boneName=jacket_sleeve_r}
    physicsBodyDefinitions [{
      sphereCollisionRadius=0
      boundBone {boneName=jacket_sleeve_r}
      boxExtents=<10 20 20>
      constraintSetup {angularLimitsMax=<5 20 10>, angularLimitsMin=<-5 -5 -10>, angularTarget=<1 0 0>}
      localJointOffset=<-3 0 0>
     }]
   }
   n38 = "Transform (Modify) Bone - Bone: jacket_collar_back"(ComponentPose=@n37, Alpha=0.8, Translation="-0.2,-2.5,0") {
    rotationMode=BMM_Additive
    rotationSpace=BCS_BoneSpace
    scaleMode=BMM_Additive
    scaleSpace=BCS_BoneSpace
    translationMode=BMM_Additive
    translationSpace=BCS_BoneSpace
    alphaBoolBlend {blendOption=Linear}
    boneToModify {boneName=jacket_collar_back}
    translation=<-0.2 -2.5>
   }
   n39 = Anim Dynamics - Bone: jacket_collar_flap_l(ComponentPose=@n38, Alpha=0.7, ExternalForce=@n36) {
    angularDampingOverride=0.8
    angularSpringConstant=10000
    bAngularSpring=true
    bGravityOverrideInSimSpace=true
    bOverrideAngularDamping=true
    numSolverIterationsPostUpdate=4
    numSolverIterationsPreUpdate=16
    simulationSpace=Actor
    alphaBoolBlend {blendOption=Linear}
    boundBone {boneName=jacket_collar_flap_l}
    physicsBodyDefinitions [{
      sphereCollisionRadius=0
      boundBone {boneName=jacket_collar_flap_l}
      boxExtents=<30 30 20>
      constraintSetup {angularTargetAxis=AxisY, twistAxis=AxisY, angularLimitsMax=<35 0 0>, angularLimitsMin=<-35 0 0>, angularTarget=<0 1 0>}
      localJointOffset=<0 4 0>
     }]
   }
   n40 = Anim Dynamics - Bone: jacket_collar_flap_r(ComponentPose=@n39, Alpha=0.7, ExternalForce=@n36) {
    angularDampingOverride=0.8
    angularSpringConstant=10000
    bAngularSpring=true
    bGravityOverrideInSimSpace=true
    bOverrideAngularDamping=true
    numSolverIterationsPostUpdate=4
    numSolverIterationsPreUpdate=16
    simulationSpace=Actor
    alphaBoolBlend {blendOption=Linear}
    boundBone {boneName=jacket_collar_flap_r}
    physicsBodyDefinitions [{
      sphereCollisionRadius=0
      boundBone {boneName=jacket_collar_flap_r}
      boxExtents=<30 30 20>
      constraintSetup {angularTargetAxis=AxisY, twistAxis=AxisY, angularLimitsMax=<35 0 0>, angularLimitsMin=<-35 0 0>, angularTarget=<0 1 0>}
      localJointOffset=<0 -4 0>
     }]
   }
   n41 = Pose Driver - Source: head!(SourcePose:Pose Link Structure=Component To Local(ComponentPose=@n40)) -> (Pose:Pose Link Structure) {
    is=title AnimGraphNode_PoseDriver
    axisLength=25
    poseAsset=&M_MED_SRDrift_elbows_pose
    evalSpaceBone {boneName=spine_05}
    onlyDriveBones [{boneName=jacket_collar_back}, {boneName=jacket_collar_l}, {boneName=jacket_collar_r}]
    poseTargets [
     {
      bApplyCustomCurve=false
      bIsHidden=false
      distanceMethod=DefaultMethod
      drivenName=head_left_pose
      functionType=DefaultFunction
      targetScale=0.9617
      boneTransforms [{targetRotation=<rot 11.0913 7.9604 63.7019>, targetTranslation=<20.6422 3.5842 0.2003>}]
      customCurve {defaultValue=3.4028234663852886e+38, postInfinityExtrap=RCCE_Constant, preInfinityExtrap=RCCE_Constant, keys []}
      targetRotation=<rot 0 0 0>
     }
     {
      bApplyCustomCurve=false
      bIsHidden=false
      distanceMethod=DefaultMethod
      drivenName=head_right_pose
      functionType=DefaultFunction
      targetScale=0.9488
      boneTransforms [{targetRotation=<rot -10.9672 7.862 -62.8452>, targetTranslation=<20.6426 3.5832 -0.1978>}]
      customCurve {defaultValue=3.4028234663852886e+38, postInfinityExtrap=RCCE_Constant, preInfinityExtrap=RCCE_Constant, keys []}
      targetRotation=<rot 0 0 0>
     }
     {
      bApplyCustomCurve=true
      bIsHidden=false
      distanceMethod=DefaultMethod
      drivenName=head_up_pose
      functionType=DefaultFunction
      targetScale=1
      boneTransforms [{targetRotation=<rot 0 -67.3521 0>, targetTranslation=<20.5299 -2.6265 0>}]
      customCurve {
       defaultValue=3.4028234663852886e+38
       postInfinityExtrap=RCCE_Constant
       preInfinityExtrap=RCCE_Constant
       keys [
        {
         arriveTangent=0
         arriveTangentWeight=0
         interpMode=RCIM_Linear
         leaveTangent=0
         leaveTangentWeight=0
         tangentMode=RCTM_Auto
         tangentWeightMode=RCTWM_WeightedNone
         time=0
         value=0
        }
        {
         arriveTangent=0
         arriveTangentWeight=0
         interpMode=RCIM_Linear
         leaveTangent=0
         leaveTangentWeight=0
         tangentMode=RCTM_Auto
         tangentWeightMode=RCTWM_WeightedNone
         time=0.6896
         value=0.9701
        }
        {
         arriveTangent=0
         arriveTangentWeight=0
         interpMode=RCIM_Linear
         leaveTangent=0
         leaveTangentWeight=0
         tangentMode=RCTM_Auto
         tangentWeightMode=RCTWM_WeightedNone
         time=1
         value=1
        }
       ]
      }
      targetRotation=<rot 0 0 0>
     }
    ]
    rBFParams {distanceMethod=Euclidean, function=Linear, radius=80}
    sourceBones [{boneName=head}]
   }
   n42 = "Constraint - jacket_collar_r (P)"(ComponentPose=Local To Component(LocalPose=@n41), Alpha=0.5, ConstraintWeights_0=1) {
    alpha=0
    alphaBoolBlend {blendOption=Linear}
    alphaScaleBiasClamp {bClampResult=true, clampMax=0.4}
    boneToModify {boneName=jacket_collar_r}
    constraintSetup [{targetBone {boneName=neck_02}}]
    constraintWeights [1]
   }
   n43 = "Transform (Modify) Bone - Bone: jacket_front_top_driver_l_MSTR"(ComponentPose=@n42, Alpha=Get crouch alpha() -> (crouch alpha:double), Translation="7,-5,0", Rotation="27.455,80,90", Scale="1,1,1") {
    rotationMode=BMM_Additive
    rotationSpace=BCS_BoneSpace
    translationMode=BMM_Additive
    translationSpace=BCS_BoneSpace
    alphaBoolBlend {blendOption=Linear}
    alphaScaleBiasClamp {bInterpResult=true, interpSpeedDecreasing=5, interpSpeedIncreasing=5}
    boneToModify {boneName=jacket_front_top_driver_l_MSTR}
    rotation=<rot 27.455 80 90>
    translation=<7 -5>
   }
   n44 = Get forward_movement() -> (forward_movement:double)
   n45 = "Transform (Modify) Bone - Bone: jacket_front_top_driver_l_MSTR"(ComponentPose=@n43, Alpha=@n44, Translation="-1.540,-1.616,1.136", Rotation="-11.420,14.346,11.581", Scale="1,1,1") {
    rotationMode=BMM_Additive
    rotationSpace=BCS_BoneSpace
    translationMode=BMM_Additive
    translationSpace=BCS_BoneSpace
    alphaBoolBlend {blendOption=Linear}
    alphaScaleBiasClamp {bInterpResult=true, interpSpeedDecreasing=5, interpSpeedIncreasing=5}
    boneToModify {boneName=jacket_front_top_driver_l_MSTR}
    rotation=<rot -11.42 14.346 11.581>
    translation=<-1.5401 -1.6157 1.136>
   }
   n46 = "Transform (Modify) Bone - Bone: jacket_front_top_driver_r_MSTR"(ComponentPose=@n45, Alpha=@n44, Translation="-1.54,1.616,-1.136", Rotation="-11.42,-14.346,11.581", Scale="1,1,1") {
    rotationMode=BMM_Additive
    rotationSpace=BCS_BoneSpace
    translationMode=BMM_Additive
    translationSpace=BCS_BoneSpace
    alphaBoolBlend {blendOption=Linear}
    alphaScaleBiasClamp {bInterpResult=true, interpSpeedDecreasing=5, interpSpeedIncreasing=5}
    boneToModify {boneName=jacket_front_top_driver_r_MSTR}
    rotation=<rot -11.42 -14.346 11.581>
    translation=<-1.54 1.616 -1.136>
   }
   n47 = "Transform (Modify) Bone - Bone: jacket_front_top_driver_r_MSTR"(ComponentPose=@n46, Alpha=Get legUp_R(), Translation="-6.216,5.872,1.108", Rotation="-21.928,-35.93,-51.234", Scale="1,1,1") {
    rotationMode=BMM_Additive
    rotationSpace=BCS_BoneSpace
    translationMode=BMM_Additive
    translationSpace=BCS_BoneSpace
    alphaBoolBlend {blendOption=Linear}
    alphaScaleBiasClamp {bInterpResult=true}
    boneToModify {boneName=jacket_front_top_driver_r_MSTR}
    rotation=<rot -21.928 -35.93 -51.234>
    translation=<-6.216 5.872 1.108>
   }
   n48 = Get legOut_L()
   n49 = "Transform (Modify) Bone - Bone: jacket_front_top_driver_l_MSTR"(ComponentPose=@n47, Alpha=@n48, Translation="6.521,-5.448,-0.000", Rotation="21.020,97.929,3.760", Scale="1,1,1") {
    rotationMode=BMM_Additive
    rotationSpace=BCS_BoneSpace
    translationMode=BMM_Additive
    translationSpace=BCS_BoneSpace
    alphaBoolBlend {blendOption=Linear}
    alphaScaleBiasClamp {bInterpResult=true}
    boneToModify {boneName=jacket_front_top_driver_l_MSTR}
    rotation=<rot 21.02 97.929 3.76>
    translation=<6.521 -5.448>
   }
   n50 = "Transform (Modify) Bone - Bone: jacket_left01_MSTR"(ComponentPose=@n49, Alpha=@n48, Translation="4.924,0.000,-3.671", Rotation="84.599,180.000,180.000", Scale="1,1,1") {
    rotationMode=BMM_Additive
    rotationSpace=BCS_BoneSpace
    translationMode=BMM_Additive
    translationSpace=BCS_BoneSpace
    alphaBoolBlend {blendOption=Linear}
    alphaScaleBiasClamp {bInterpResult=true}
    boneToModify {boneName=jacket_left01_MSTR}
    rotation=<rot 84.599 180 180>
    translation=<4.9237 0 -3.6714>
   }
   n51 = Get legOut_R()
   n52 = "Transform (Modify) Bone - Bone: jacket_front_top_driver_r_MSTR"(ComponentPose=@n50, Alpha=@n51, Translation="6.521,5.448,0", Rotation="21.02,-97.929,-3.76", Scale="1,1,1") {
    rotationMode=BMM_Additive
    rotationSpace=BCS_BoneSpace
    translationMode=BMM_Additive
    translationSpace=BCS_BoneSpace
    alphaBoolBlend {blendOption=Linear}
    alphaScaleBiasClamp {bInterpResult=true}
    boneToModify {boneName=jacket_front_top_driver_r_MSTR}
    rotation=<rot 21.02 -97.929 -3.76>
    translation=<6.521 5.448>
   }
   n53 = "Transform (Modify) Bone - Bone: jacket_left04_MSTR"(ComponentPose=@n52, Alpha=@n51, Translation="4.924,0.000,-3.671", Rotation="84.599,-180,-180", Scale="1,1,1") {
    rotationMode=BMM_Additive
    rotationSpace=BCS_BoneSpace
    translationMode=BMM_Additive
    translationSpace=BCS_BoneSpace
    alphaBoolBlend {blendOption=Linear}
    alphaScaleBiasClamp {bInterpResult=true}
    boneToModify {boneName=jacket_left04_MSTR}
    rotation=<rot 84.599 -180 -180>
    translation {x=4.9237, z=-3.6714}
   }
   Save cached pose 'Dynamics'(Pose=Component To Local(ComponentPose=@n53)) {cacheName=Dynamics}
   n54 = Get bIsMoving2D()
   n55 = "Transform (Modify) Bone - Bone: jacket_back_right01"!(
    ComponentPose=Local To Component(LocalPose=Use cached pose 'Dynamics'())
    bAlphaBoolEnabled:bool=AND Boolean(A=Get bIsCrouching(), B=NOT Boolean(A=@n54))
    Translation="0,-3,0"
    Rotation="45,15,0"
    Scale="1,1,1"
   ) -> (Pose) {
    alphaInputType=Bool
    rotationMode=BMM_Additive
    rotationSpace=BCS_BoneSpace
    translationMode=BMM_Additive
    translationSpace=BCS_BoneSpace
    alphaBoolBlend {blendInTime=0.1, blendOption=Linear, blendOutTime=0.1}
    boneToModify {boneName=jacket_back_right01}
    rotation {pitch=45, yaw=15}
    translation {y=-3}
   }
   n56 = "Transform (Modify) Bone - Bone: attach"!(
    ComponentPose=@n55
    bAlphaBoolEnabled:bool=OR Boolean!(A=Get bIsDBNO(), B=Get bIsJumping(), C:bool=Get bIsFalling()) -> (ReturnValue)
    Translation="0,0,0"
    Rotation="0,0,0"
    Scale="1,1,1"
   ) -> (Pose) {
    alphaInputType=Bool
    rotationMode=BMM_Additive
    rotationSpace=BCS_BoneSpace
    translationMode=BMM_Additive
    translationSpace=BCS_BoneSpace
    alphaBoolBlend {blendInTime=0.1, blendOption=Linear, blendOutTime=0.1}
    boneToModify {boneName=attach}
   }
   n57 = RigidBody!(
    ComponentPose=@n56
    Alpha="To Float (Boolean)"(InBool=Get standardGravityAlpha() -> (standardGravityAlpha:bool))
    OverrideWorldGravity:Vector=Get GravityOverride()
   ) -> (Pose) {
    bClampLinearTranslationLimitToRefPose=true
    bOverrideWorldGravity=true
    overridePhysicsAsset=&/M_MED_SRDrift_tier0506_Dyn_jacket_Update
    simulationSpace=BaseBoneSpace
    alphaBoolBlend {blendOption=Linear}
    alphaScaleBiasClamp {bInterpResult=true}
    baseBoneRef {boneName=dyn_root_simspace}
    componentAppliedLinearAccClamp=<3000 3000 3000>
    componentLinearAccScale {z=0.1}
    componentLinearVelScale {z=0.1}
    overrideWorldGravity {z=-980}
    simSpaceSettings {
     maxAngularAcceleration=1000
     maxAngularVelocity=3
     maxLinearAcceleration=700
     maxLinearVelocity=500
     velocityScaleZ=0
     worldAlpha=0.25
     externalLinearDragV=<0.2 0.2>
    }
   }
   n58 = RigidBody(ComponentPose=@n57, Alpha="To Float (Boolean)"(InBool=Get altGravityAlpha() -> (altGravityAlpha:bool)), OverrideWorldGravity="0,0,150") {
    bClampLinearTranslationLimitToRefPose=true
    bOverrideWorldGravity=true
    overridePhysicsAsset=&M_MED_SRDrift_tier0506_Dyn_jacket_Update
    simulationSpace=BaseBoneSpace
    alphaBoolBlend {blendOption=Linear}
    alphaScaleBiasClamp {bInterpResult=true}
    baseBoneRef {boneName=dyn_root_simspace}
    componentAppliedLinearAccClamp=<3000 3000 3000>
    componentLinearAccScale {z=0.1}
    componentLinearVelScale {z=0.1}
    overrideWorldGravity {z=-980}
    simSpaceSettings {
     maxAngularAcceleration=1000
     maxAngularVelocity=3
     maxLinearAcceleration=700
     maxLinearVelocity=500
     velocityScaleZ=0
     worldAlpha=0.25
     externalLinearDragV=<0.2 0.2>
    }
   }
   n59 = Component To Local(ComponentPose=@n58)
   n60 = "Constraint - jacket_front_top_driver_l_MSTR (P)"(ComponentPose=Local To Component(LocalPose=Use cached pose 'Dynamics'()), Alpha=0.5, ConstraintWeights_0=1) {
    alpha=0
    alphaBoolBlend {blendOption=Linear}
    alphaScaleBiasClamp {bClampResult=true, clampMax=0.4}
    boneToModify {boneName=jacket_front_top_driver_l_MSTR}
    constraintSetup [{targetBone {boneName=thigh_l}}]
    constraintWeights [1]
   }
   n61 = "Constraint - jacket_front_left02 (P)"(ComponentPose=@n60, Alpha=0.5, ConstraintWeights_0=1) {
    alpha=0
    alphaBoolBlend {blendOption=Linear}
    alphaScaleBiasClamp {bClampResult=true, clampMax=0.4}
    boneToModify {boneName=jacket_front_left02}
    constraintSetup [{targetBone {boneName=thigh_l}}]
    constraintWeights [1]
   }
   n62 = "Transform (Modify) Bone - Bone: jacket_front_top_driver_l"!(ComponentPose=@n61, Alpha=0.8, Translation="-0.000,-0.000,-2", Rotation="10.000,0.000,0.000") -> (Pose) {
    rotationMode=BMM_Additive
    rotationSpace=BCS_BoneSpace
    scaleMode=BMM_Additive
    scaleSpace=BCS_BoneSpace
    translationMode=BMM_Additive
    translationSpace=BCS_BoneSpace
    alphaBoolBlend {blendOption=Linear}
    boneToModify {boneName=jacket_front_top_driver_l}
    rotation {pitch=10}
    translation {z=-2}
   }
   n63 = "Transform (Modify) Bone - Bone: jacket_left01"!(ComponentPose=@n62, Alpha=0.8, Translation="-0.000,-0.000,-2", Rotation="10.000,0.000,0.000") -> (Pose) {
    rotationMode=BMM_Additive
    rotationSpace=BCS_BoneSpace
    scaleMode=BMM_Additive
    scaleSpace=BCS_BoneSpace
    translationMode=BMM_Additive
    translationSpace=BCS_BoneSpace
    alphaBoolBlend {blendOption=Linear}
    boneToModify {boneName=jacket_left01}
    rotation {pitch=10}
    translation {z=-2}
   }
   n64 = "Constraint - jacket_front_top_driver_r_MSTR (P)"(ComponentPose=@n63, Alpha=0.5, ConstraintWeights_0=1) {
    alpha=0
    alphaBoolBlend {blendOption=Linear}
    alphaScaleBiasClamp {bClampResult=true, clampMax=0.4}
    boneToModify {boneName=jacket_front_top_driver_r_MSTR}
    constraintSetup [{targetBone {boneName=thigh_r}}]
    constraintWeights [1]
   }
   n65 = "Constraint - jacket_front_right02 (P)"(ComponentPose=@n64, Alpha=0.5, ConstraintWeights_0=1) {
    alpha=0
    alphaBoolBlend {blendOption=Linear}
    alphaScaleBiasClamp {bClampResult=true, clampMax=0.4}
    boneToModify {boneName=jacket_front_right02}
    constraintSetup [{targetBone {boneName=thigh_r}}]
    constraintWeights [1]
   }
   n66 = "Transform (Modify) Bone - Bone: jacket_front_top_driver_r"!(ComponentPose=@n65, Alpha=0.8, Translation="0,0.000,-3", Rotation="10.000,0.000,0.000") -> (Pose) {
    rotationMode=BMM_Additive
    rotationSpace=BCS_BoneSpace
    scaleMode=BMM_Additive
    scaleSpace=BCS_BoneSpace
    translationMode=BMM_Additive
    translationSpace=BCS_BoneSpace
    alphaBoolBlend {blendOption=Linear}
    boneToModify {boneName=jacket_front_top_driver_r}
    rotation {pitch=10}
    translation {z=-3}
   }
   n67 = "Transform (Modify) Bone - Bone: jacket_left04"!(ComponentPose=@n66, Alpha=0.8, Translation="-0.000,-0.000,-2", Rotation="10.000,0.000,0.000") -> (Pose) {
    rotationMode=BMM_Additive
    rotationSpace=BCS_BoneSpace
    scaleMode=BMM_Additive
    scaleSpace=BCS_BoneSpace
    translationMode=BMM_Additive
    translationSpace=BCS_BoneSpace
    alphaBoolBlend {blendOption=Linear}
    boneToModify {boneName=jacket_left04}
    rotation {pitch=10}
    translation {z=-2}
   }
   n68 = Component To Local(ComponentPose=@n67)
   n69 = Get bRigidBodyEnabled()
   n70 = Get WeaponKickupExternalForce()
   n71 = Anim Dynamics - Bone: hair_dyn(
    ComponentPose=Local To Component(LocalPose=Blend Poses by bool(BlendPose_0=@n59, BlendPose_1=@n68, bActiveValue=@n69) {blendType=Linear})
    Alpha=0.6
    ExternalForce=@n70
   ) {
    angularDampingOverride=0.99
    angularSpringConstant=3000
    bAngularSpring=true
    bGravityOverrideInSimSpace=true
    bOverrideAngularDamping=true
    numSolverIterationsPostUpdate=4
    numSolverIterationsPreUpdate=16
    alphaBoolBlend {blendOption=Linear}
    boundBone {boneName=hair_dyn}
    physicsBodyDefinitions [{
      sphereCollisionRadius=0
      boundBone {boneName=hair_dyn}
      boxExtents=<15 18 18>
      constraintSetup {angularLimitsMax=<5 20 20>, angularLimitsMin=<-5 -20 -20>, angularTarget=<1 0 0>}
      localJointOffset=<-3 0 0>
     }]
   }
   n72 = Anim Dynamics - Bone: collar_l(ComponentPose=@n71, Alpha=0.2, ExternalForce=@n70) {
    angularDampingOverride=0.8
    angularSpringConstant=500
    bAngularSpring=true
    bGravityOverrideInSimSpace=true
    bOverrideAngularDamping=true
    numSolverIterationsPostUpdate=4
    numSolverIterationsPreUpdate=16
    alphaBoolBlend {blendOption=Linear}
    boundBone {boneName=collar_l}
    componentLinearAccScale=<1 1 1>
    componentLinearVelScale=<2 2 2>
    physicsBodyDefinitions [{
      sphereCollisionRadius=0
      boundBone {boneName=collar_l}
      boxExtents=<15 18 18>
      constraintSetup {angularLimitsMax=<5 20 20>, angularLimitsMin=<-5 -20 -20>, angularTarget=<1 0 0>, linearAxesMax=<0 0 1>, linearAxesMin=<0 0 -1>}
      localJointOffset=<-3 0 0>
     }]
   }
   n73 = Anim Dynamics - Bone: collar_r(ComponentPose=@n72, Alpha=0.2, ExternalForce=@n70) {
    angularDampingOverride=0.8
    angularSpringConstant=500
    bAngularSpring=true
    bGravityOverrideInSimSpace=true
    bOverrideAngularDamping=true
    numSolverIterationsPostUpdate=4
    numSolverIterationsPreUpdate=16
    alphaBoolBlend {blendOption=Linear}
    boundBone {boneName=collar_r}
    componentLinearAccScale=<1 1 1>
    componentLinearVelScale=<2 2 2>
    physicsBodyDefinitions [{
      sphereCollisionRadius=0
      boundBone {boneName=collar_r}
      boxExtents=<15 18 18>
      constraintSetup {angularLimitsMax=<5 20 20>, angularLimitsMin=<-5 -20 -20>, angularTarget=<1 0 0>, linearAxesMax=<0 0 1>, linearAxesMin=<0 0 -1>}
      localJointOffset=<-3 0 0>
     }]
   }
   Save cached pose 'PreScale Cache'(Pose=Component To Local(ComponentPose=@n73)) {cacheName=PreScale Cache}
   Get ThighRightAlpha() -> (ThighRightAlpha:float)
   Get ThighLeftAlpha() -> (ThighLeftAlpha:float)
  ]}
 EventGraph {nodes [
   n1 = Get drift_clothing_sim_interactor() -> (drift_clothing_sim_interactor:Clothing Simulation Interactor Nv Object Reference)
   SetAnimDriveSpringStiffness!(execute, self:Clothing Simulation Interactor Object Reference=@n1, InStiffness:float=0) -> (then)
   > SetAnimDriveDamperStiffness!(execute, self:Clothing Simulation Interactor Nv Object Reference=@n1, InStiffness:float=0) -> (then)
   SetAnimDriveSpringStiffness(self=@n1, InStiffness=20)
   > SetAnimDriveDamperStiffness(self=@n1, InStiffness=10)
   Event OnBeginSkydiving()
   > Reroute Node()
   Event BlueprintInitializeAnimation()
   > Set Body_Mesh(Body_Mesh=GetPartSkeletalMeshComponent())
   Event OnEndCrouching()
   > Set crouch alpha(crouch alpha:double=0) -> (Output_Get:double)
   Event OnBeginCrouching()
   > Set crouch alpha(crouch alpha=1)
   Event OnEndSkydiving()
   > Reroute Node()
   n2 = Event BlueprintUpdateAnimation()
   > n3 = Sequence!(execute) -> (then_0, then_1, then_2:Exec, then_3:Exec, then_4:Exec)
   > Get!(execute:Exec) -> (then:Exec, else:Exec, Body_Mesh:Skeletal Mesh Component Object Reference)
   > Set delta_time(delta_time:double=@n2.DeltaTimeX) -> (Output_Get:double)
   > Set legUp_L(legUp_L:double=GetAlphaFromReferenceTransformDelta(SkeletalMeshComponent=Get Body_Mesh(), BoneName=thigh_l, BaseName=pelvis, InputRange=(X=-50,Y=-90), TwistAxis="0,0,0")) -> (Output_Get:double)
   > Set legUp_R(legUp_R:double=GetAlphaFromReferenceTransformDelta(SkeletalMeshComponent=Get Body_Mesh(), BoneName=thigh_r, BaseName=pelvis, InputRange=(X=-50,Y=-90), TwistAxis="0,0,0")) -> (Output_Get:double)
   n4 = Get Body_Mesh()
   > Set legOut_R(legOut_R:double=GetAlphaFromReferenceTransformDelta(SkeletalMeshComponent=@n4, BoneName=thigh_r, BaseName=pelvis, DeltaComponentType=EulerY, InputRange=(X=30,Y=50), TwistAxis="0,1,0")) -> (Output_Get:double)
   > Set legOut_L(legOut_L:double=GetAlphaFromReferenceTransformDelta(SkeletalMeshComponent=@n4, BoneName=thigh_l, BaseName=pelvis, DeltaComponentType=EulerY, InputRange=(X=30,Y=50), TwistAxis="0,1,0")) -> (Output_Get:double)
   n5 = PlayerMovement(execute=@n3.then_1, AnimInstance=Self-Reference(), XDotProductCoefficient=1, YDotProductCoefficient=1, ZDotProductCoefficient=1)
   > Set forward_movement(forward_movement:double=MapRangeClamped(Value=@n5.YMovement, InRangeA=0, InRangeB=1)) -> (Output_Get:double)
   n6 = Gravity Override State Function DEPRECATED(
    execute=@n3.then_2
    CharacterPartAnimInstance=Self-Reference()
    GameTimeInSeconds=GetGameTimeInSeconds()
    GravityOverrideParams=Get Gravity Override Params()
   )
   > Set GravityOverride(GravityOverride=@n6)
   n7 = ComplexSimSpaceFunction DEPRECATED!(
    execute=@n3.then_3
    CharacterPartAnimInstance:Custom Character Part Anim Instance Object Reference=Self-Reference()
    InterpSpeedMax:double=9
    PawnSpeedMax:double=20
    RInterpErrorTolerance:double=1
    DeltaTime:double=0
    DebugDraw:bool=false
    sim_space_rotation_in:Rotator=Get sim_space_rotation() -> (sim_space_rotation:Rotator)
    inComponentSpaceIn:bool=Get InComponentSpace() -> (InComponentSpace:bool)
    inComponentSpaceLastFrameIn:bool=Get InComponentSpaceLastFrame() -> (InComponentSpaceLastFrame:bool)
    lockedOnSimSpaceIn:bool=Get LockedOnSimSpace() -> (LockedOnSimSpace:bool)
    InterpSpeedIn:double=Get InterpSpeed() -> (InterpSpeed:double)
   ) -> (
    then
    sim_space_rotation:Rotator
    inComponentSpaceOut:bool
    inComponentSpaceLastFrameOut:bool
    lockedOnSimSpaceOut:bool
    InterpSpeedOut:double
   )
   > Set sim_space_rotation(sim_space_rotation:Rotator=@n7) -> (Output_Get:Rotator)
   > Set InComponentSpace(InComponentSpace:bool=@n7.inComponentSpaceOut) -> (Output_Get:bool)
   > Set InComponentSpaceLastFrame(InComponentSpaceLastFrame:bool=@n7.inComponentSpaceLastFrameOut) -> (Output_Get:bool)
   > Set LockedOnSimSpace(LockedOnSimSpace:bool=@n7.lockedOnSimSpaceOut) -> (Output_Get:bool)
   > Set InterpSpeed(InterpSpeed:double=@n7.InterpSpeedOut) -> (Output_Get:double)
   n8 = Get bIsSkyDiving()
   n9 = Get bIsFalling()
   n10 = Get bIsJumping()
   n11 = Branch(execute=@n3.then_4, Condition=AND Boolean(A=OR Boolean!(A=@n9, B=@n10) -> (ReturnValue), B=NOT Boolean(A=@n8)))
   > Set altGravityAlpha(altGravityAlpha:bool=true) -> (Output_Get:bool)
   > Set standardGravityAlpha(standardGravityAlpha:bool=false) -> (Output_Get:bool)
   Set standardGravityAlpha(execute=@n11.else, standardGravityAlpha=true)
   > Set altGravityAlpha(altGravityAlpha=false)
   Set random_stream(random_stream:Random Stream Structure=MakeRandomStream!(InitialSeed:int=0) -> (ReturnValue:Random Stream Structure)) -> (Output_Get:Random Stream Structure)
   > n12 = Cast To ClothingSimulationInteractorNv!(
    execute
    Object=GetClothingSimulationInteractor!(self:Skeletal Mesh Component Object Reference=GetPartSkeletalMeshComponent(), ClothingAssetName:Name=None) -> (ReturnValue:Clothing Simulation Interactor Object Reference)
   ) -> (then, CastFailed, AsClothing Simulation Interactor Nv:Clothing Simulation Interactor Nv Object Reference) {is=title K2Node_DynamicCast}
   > Set drift_clothing_sim_interactor(drift_clothing_sim_interactor:Clothing Simulation Interactor Nv Object Reference=@n12) -> (Output_Get:Clothing Simulation Interactor Nv Object Reference)
   > Set GetPlayerListUsesCustomCharacterParts(GetPlayerListUsesCustomCharacterParts:bool=GetPlaylistUsesCustomCharacterParts!() -> (ReturnValue:bool)) -> (Output_Get:bool)
   GetOwnerSkeletalMeshComponent()
   Set Anim_Inst(Anim_Inst:Anim Instance Object Reference=GetAnimInstance()) -> (Output_Get:Anim Instance Object Reference)
  ]}
]
variables [
 altGravityAlpha {type=boolean}
 Anim_Inst {type=AnimInstance, default=null}
 Body_Mesh {type=SkeletalMeshComponent}
 crouch alpha {type=number}
 delta_time {type=number}
 drift_clothing_sim_interactor {type=ClothingSimulationInteractorNv, default=null}
 forward_movement {type=number}
 GetPlayerListUsesCustomCharacterParts {type=boolean}
 Gravity Override Params {
  type=GravityOverrideParamsStruct
  default {
   skydive_DivingDown {gravityOverride=<0 0 0>}
   skydive_FloatingDown {gravityOverride=<0 0 0>, windAmplitude=<100 -250 200>, windOffset=<0 -300 200>}
   skydive_GoingUp {gravityOverride=<0 0 -100>}
   swimming {gravityOverride=<0 0 0>}
  }
 }
 GravityOverride {type=Vector}
 InComponentSpace {type=boolean}
 InComponentSpaceLastFrame {type=boolean}
 InterpSpeed {type=number}
 legOut_L {type=number}
 legOut_R {type=number}
 legUp_L {type=number}
 legUp_R {type=number}
 LockedOnSimSpace {type=boolean}
 random_stream {type=RandomStream, default {initialSeed=0, seed=0}}
 sim_space_rotation {type=Rotator}
 standardGravityAlpha {type=boolean}
]
